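# Parent wheel material for the FakeSugarSE Infinity vehicle cosmetic, adding an animated, tinted tire glow to MF_WheelParent_Simple.
# An emissive mask splits the glow: R pulses on a clamped cosine, G stays steady, B pulses through panning, rotated noise.
# Vertex red minus the tire mask's green blends this glow over the base emissive.
# The result passes through the shared vehicle damage and de-rez transition functions.
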Material /VehicleCosmetics/Wheels/FakeSugarSE_INF/Materials/M_WheelParent_FakeSugarSE
graph {
 properties {
  bAutomaticallySetUsageInEditor=false
  bUseMaterialAttributes=true
  bUsedWithMeshDeformer=true
  bUsedWithSkeletalMesh=true
  pixelDepthOffsetMode=PDOM_Legacy
  previewMesh=/VehicleCosmetics/Wheels/FakeSugarSE_INF/Meshes/SK_Wheel_FakeSugarSeInfinity
  usedShadingModels=Default Lit
 }
 nodes [
  n1 = MaterialFunctionCall() {materialFunction=&MF_WheelParent_Simple}
  n2 = GetMaterialAttributes(None=@n1.MaterialAttributes) {attributeGetTypes [#1]}
  n3 = LinearInterpolate(A=@n2.Emissive Color, B=NamedRerouteUsage(), Alpha=Subtract(A=@n1.VertexColorR, B=ComponentMask(None=@n1.TireRGBAMask) {g=true}))
  MaterialFunctionCall(In=MaterialFunctionCall(Material=SetMaterialAttributes(MaterialAttributes=@n2.MaterialAttributes, Emissive Color=@n3) {attributeSetTypes [#1]}) {materialFunction=&MF_BaseVehicle_Damage}.Result) {materialFunction=&MF_VehicleDeRez_Transition}
  n4 = TextureCoordinate() {uTiling=2}
  n5 = Time()
  n6 = TextureSample(UVs=Rotator(Coordinate=Panner(Coordinate=@n4, Time=@n5) {speedX=0.5}, Time=Constant() {r=1.5708}) {centerX=0, centerY=0, speed=1}) {samplerSource=SSM_Wrap_WorldGroupSettings, texture=/VehicleCosmetics/Wheels/_SharedTextures/FX/Universal_Noise_03_RGB}
  n7 = Clamp(None=Cosine(None=Multiply(A=@n5) {constB=0.25}) {period=3}) {minDefault=0.2}
  n8 = TextureObjectParameter() {
   group=5_Special Effects
   parameterName=Emissive Masks
   samplerType=SAMPLERTYPE_LinearColor
   sortPriority=0
   texture=/VehicleCosmetics/Wheels/FakeSugarSE_INF/Textures/T_FakeSugarSeInfinity_Tire_E
  }
  n9 = TextureSample(Tex=@n8) {
   samplerSource=SSM_Wrap_WorldGroupSettings
   samplerType=SAMPLERTYPE_LinearColor
   texture=&T_FakeSugarSeInfinity_Tire_E
  }
  n10 = Add(A=Multiply(A=Multiply(A=@n6.B, B=@n7), B=@n9.B), B=Add(A=Multiply(A=@n7, B=@n9.R) {constB=4}, B=Multiply(A=@n9.G) {constB=4}))
  n11 = ScalarParameter() {defaultValue=3, group=5_Special Effects, parameterName=Tire Emissive Brightness, sortPriority=0}
  n12 = VectorParameter() {group=5_Special Effects, parameterName=Emissive Color, sortPriority=0, defaultValue=<rgba 0 0.2307 1 1>}
  NamedRerouteDeclaration(None=Multiply(A=Multiply(A=@n10, B=@n11), B=@n12.RGB)) {name=Emissive, nodeColor=<rgba 1 0.75 0 1>}
 ]
}
