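# Unlit vehicle cosmetic material for the FakeSugar "INF" wheel, drawing an animated infinity-style emissive effect on skeletal meshes.
# MF_Infinity_OffsetFunction_Spring processes a gradient ring texture on UV1, driven by Speed and Distance parameters.
# Its red channel is tinted by Paint Color and its green by a lifted, brighter variant, combined with Max and scaled by brightness.
# The resulting emissive attributes pass through the shared vehicle damage and de-rez transition functions.
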Material /VehicleCosmetics/Wheels/FakeSugarSE_INF/Materials/M_Wheel_FakeSugar_INF
graph {
 properties {
  bAutomaticallySetUsageInEditor=false
  bUseMaterialAttributes=true
  bUsedWithMeshDeformer=true
  bUsedWithSkeletalMesh=true
  previewMesh=/VehicleCosmetics/Wheels/Glitch/Meshes/SK_Wheel_Glitch
  shadingModel=MSM_Unlit
  usedShadingModels=Unlit
 }
 nodes [
  n1 = VectorParameter() {group=00 Base, parameterName=Paint Color, sortPriority=0, defaultValue=<rgba 1 0.314 0 1>}
  n2 = MaterialFunctionCall!(
   Texture=TextureObjectParameter() {parameterName=Infinity Texture, sortPriority=1, texture=/VehicleCosmetics/Wheels/_SharedTextures/FX/GradientRing_T}
   UVs=TextureCoordinate() {coordinateIndex=1}
   RotateUVs
   GradientFalloff=StaticBool()
   Speed=ScalarParameter() {defaultValue=3, group=00 Base, parameterName=Speed, sortPriority=1}
   Distance=ScalarParameter() {defaultValue=0.8, group=00 Base, parameterName=Distance, sortPriority=2}
   GradientDarkness
  ) {materialFunction=&/MF_Infinity_OffsetFunction_Spring}
  n3 = ComponentMask(None=@n2.Result) {r=true}
  n4 = Multiply(A=ConstantBiasScale(None=@n1.RGB) {bias=0.2, scale=1}) {constB=2}
  n5 = ComponentMask(None=@n2.Result) {g=true}
  n6 = Multiply(
   A=Max(A=Multiply(A=@n1.RGB, B=@n3), B=Multiply(A=@n4, B=@n5))
   B=ScalarParameter() {defaultValue=1, group=00 Base, parameterName=Wheel Emissive Brightness, sortPriority=3}
  )
  MaterialFunctionCall(In=MaterialFunctionCall(Material=SetMaterialAttributes(Emissive Color=@n6) {attributeSetTypes [#1]}) {materialFunction=&MF_BaseVehicle_Damage}.Result) {materialFunction=&MF_VehicleDeRez_Transition}
 ]
}
